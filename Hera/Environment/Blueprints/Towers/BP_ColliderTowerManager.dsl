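; Manager for the collider towers that plays a looping, one-after-another light boost across the tower array.
; After startup delays it marks itself invulnerable, caches the last tower index, and starts the animation on the server only.
; ColliderTowersAnim calls BoostLight on the current tower, advances or wraps the index, then loops back through a short Delay rather than a timer.
; An invalid tower reference prints a debug string and ends the loop.
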(event EventBeginPlay
  (Utilities|FlowControl|Delay 5.0)
  (Variables|BuildingActor|SetbIsInvulnerable true)
  (Variables|Default|SetNumTowers (Utilities|Array|LastIndex (Variables|Default|GetAllColliderTowers)))
  (Utilities|FlowControl|Delay 1.0)
  (Networking|SwitchHasAuthority
    (:Authority
      (CallFunction|ColliderTowersAnim))))

(event Custom|ColliderTowersAnim
  (bind _currenttowerindex (Variables|Default|GetCurrentTowerIndex))
  (bind _allcollidertowers (Variables|Default|GetAllColliderTowers))
  (bind _output (Utilities|Array|Get(acopy) _allcollidertowers _currenttowerindex))
  (bind _currenttowerindex_2 (Variables|Default|GetCurrentTowerIndex))
  (bind _returnvalue (+ _currenttowerindex_2 1))
  (Networking|SwitchHasAuthority
    (:Authority
      (Utilities|IsValid _output :!label _label
        (:Is%20Valid
          (Class|BPColliderTowerA|BoostLight _output)
          (if (== (Variables|Default|GetCurrentTowerIndex) (Variables|Default|GetNumTowers))
            _
            (elif (> _returnvalue (Variables|Default|GetNumTowers))
              _
              (else
                (Variables|Default|SetCurrentTowerIndex _returnvalue)
                (Utilities|FlowControl|Delay 0.5 :!label _label_2)
                (goto _label))))
          (Variables|Default|SetCurrentTowerIndex 0)
          (goto _label_2))
        (:Is%20Not%20Valid
          (Development|PrintString "Hello"))))))

(fn ConstructionScript)
