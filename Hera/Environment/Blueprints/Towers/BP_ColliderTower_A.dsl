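; BP_ColliderTower_A is a tower Blueprint whose Satellite_Flare material shows a brief boosted-light state.
; BoostLight plays the GhostRocks activate cue, sets BoostedLight true, then clears it after a short delay; BeginPlay also clears it.
; OnRep_BoostedLight swaps between A (boosted) and B (unboosted) flare materials, with separate variants when IsPrimaryTower is set.
; The construction script applies the A material for the tower's type.
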(event EventBeginPlay
  (Variables|Default|SetBoostedLight false))

(event Custom|BoostLight
  (Audio|Components|Audio|Play (Variables|Default|GetGhostRocks_Activate_Cue) 0.0)
  (Variables|Default|SetBoostedLight true)
  (Utilities|FlowControl|Delay 0.65)
  (Variables|Default|SetBoostedLight false))

(fn OnRep_BoostedLight
  (bind _satellite_flare (Variables|Default|GetSatellite_Flare))
  (bind _satellite_flare_2 (Variables|Default|GetSatellite_Flare))
  (if (Variables|Default|GetBoostedLight)
    (if (Variables|Default|GetIsPrimaryTower)
      (Rendering|Material|SetMaterial _satellite_flare 0 "/Hera/Environment/Blueprints/Towers/MI_Tower_CamFacing_A_A.MI_Tower_CamFacing_A_A")
      (else
        (Rendering|Material|SetMaterial _satellite_flare 0 "/Hera/Environment/Blueprints/Towers/MI_Tower_CamFacing_A.MI_Tower_CamFacing_A")))
    (elif (Variables|Default|GetIsPrimaryTower)
      (Rendering|Material|SetMaterial _satellite_flare_2 0 "/Hera/Environment/Blueprints/Towers/MI_Tower_CamFacing_B_B.MI_Tower_CamFacing_B_B")
      (else
        (Rendering|Material|SetMaterial _satellite_flare_2 0 "/Hera/Environment/Blueprints/Towers/MI_Tower_CamFacing_B.MI_Tower_CamFacing_B")))))

(fn ConstructionScript
  (bind _satellite_flare (Variables|Default|GetSatellite_Flare))
  (if (Variables|Default|GetIsPrimaryTower)
    (Rendering|Material|SetMaterial _satellite_flare 0 "/Hera/Environment/Blueprints/Towers/MI_Tower_CamFacing_A_A.MI_Tower_CamFacing_A_A")
    (else
      (Rendering|Material|SetMaterial _satellite_flare 0 "/Hera/Environment/Blueprints/Towers/MI_Tower_CamFacing_A.MI_Tower_CamFacing_A"))))
